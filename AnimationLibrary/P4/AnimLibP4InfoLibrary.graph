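# Editor utility library whose GetP4Infos function gathers every AnimLibP4Info asset in the project for the Animation Library.
# It queries the Asset Registry by that class's path, including subclasses.
# Each returned asset is loaded via GetAsset, cast to AnimLibP4Info and added uniquely to a results array, which is returned when the loop completes.
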EditorUtilityBlueprint /AnimationLibrary/P4/AnimLibP4InfoLibrary
graphs [GetP4Infos {nodes [
   GetP4Infos() {is=title K2Node_FunctionEntry}
   > n1 = GetAssetsByClass!(
    execute:Exec
    self:Object Reference="To Object (Interface)"!(Interface:"Interface Interface (by ref)"=GetAssetRegistry()) -> (ReturnValue:Object Reference)
    ClassPathName:Top Level Asset Path Structure="Get Class Path Name (Top Level Asset Path)"!(Class:Object Class Reference=&/AnimLibP4Info.AnimLibP4Info_C) -> (ReturnValue:Top Level Asset Path Structure) {is=title K2Node_CallFunction}
    bSearchSubClasses:bool=true
   ) -> (then:Exec, OutAssetData:Array of Asset Data Structures, ReturnValue:bool) {is=title K2Node_Message}
   > n2 = For Each Loop(Array:Array of Asset Data Structures=@n1) -> (Array Element:Asset Data Structure)
   Return Node!(
    execute=@n2.Completed
    AsAnim Lib P4Info:Array of Anim Lib P4Info Object References=Get Results() -> (Results:Array of Anim Lib P4Info Object References)
   )
   > n3 = Cast To AnimLibP4Info!(execute, Object=GetAsset(InAssetData=@n2)) -> (then, CastFailed, AsAnim Lib P4Info:Anim Lib P4Info Object Reference) {is=title K2Node_DynamicCast}
   > Add Unique(TargetArray:Array of Anim Lib P4Info Object References=Get Results(), NewItem:"Anim Lib P4Info Object Reference (by ref)"=@n3)
  ]}]
variables []
